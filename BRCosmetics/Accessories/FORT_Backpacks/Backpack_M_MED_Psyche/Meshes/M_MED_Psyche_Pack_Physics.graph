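# Physics asset for the Psyche backpack cosmetic, apparently set up to move two liquid elements, left and right, on the pack.
# Each side pairs a thin, flat box body on the liquid bone with a small box body on its master bone.
# One constraint per side joins the pair, locking twist and setting narrow swing limit angles.
# A tiny capsule covers the pack bone, which no constraint links to either pair.
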PhysicsAsset /BRCosmetics/Accessories/FORT_Backpacks/Backpack_M_MED_Psyche/Meshes/M_MED_Psyche_Pack_Physics
bodies {
 dyn_liquid_psyche_lft [{shapeType=Box, extentX=11.01, extentY=11.01, extentZ=1.01, center=<1.1768 0 0>, rotation=<rot 90 0 0>}]
 dyn_liquid_psyche_rgt [{shapeType=Box, extentX=11.01, extentY=11.01, extentZ=1.01, center=<1.1768 0 0>, rotation=<rot 90 0 0>}]
 dyn_master_psyche_lft [{shapeType=Box, extentX=1.01, extentY=1.01, extentZ=1.01, shapeName=dyn_master_psyche_lft_box, rotation=<rot 0 -90 0>}]
 dyn_master_psyche_rgt [{shapeType=Box, extentX=1.01, extentY=1.01, extentZ=1.01, shapeName=dyn_master_psyche_rgt_box, rotation=<rot 0 -90 0>}]
 pack [{shapeType=Capsule, length=0.505, radius=0.505, shapeName=pack_capsule, rotation=<rot 0 -90 0>}]
}
constraints [
 {
  bone1Name=dyn_liquid_psyche_lft
  bone2Name=dyn_master_psyche_lft
  swing1LimitDegrees=20
  swing2LimitDegrees=20
  twistLimitDegrees=45
  twistMotion=Locked
 }
 {
  bone1Name=dyn_liquid_psyche_rgt
  bone2Name=dyn_master_psyche_rgt
  swing1LimitDegrees=20
  swing2LimitDegrees=20
  twistLimitDegrees=45
  twistMotion=Locked
 }
]
