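# Physics asset for the Psyche backpack mesh, providing capsule bodies and joints for its left and right "black fluid" bone chains.
# A small capsule on the pack bone anchors two chains: four bodies on the left and three on the right, skipping some numbered bones.
# Each link is constrained to the one before it with the same swing limits and a very tight twist limit, so the strands bend but barely twist.
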PhysicsAsset /Game/Accessories/FORT_Backpacks/Backpack_M_MED_Psyche/Meshes/M_MED_Psyche_Pack_Physics_2
bodies {
 dyn_black_fluid_lft_01 [{
   shapeType=Capsule
   length=3.1999
   radius=2.5844
   shapeName=dyn_black_fluid_lft_01_capsule
   center=<2.6345 0 0>
   rotation=<rot 90 -5.0511 -5.0511>
  }]
 dyn_black_fluid_lft_02 [{
   shapeType=Capsule
   length=4.1999
   radius=2.5844
   shapeName=dyn_black_fluid_lft_01_capsule
   center=<5.1855 -0.0002 0>
   rotation=<rot 90 2.5305 2.5305>
  }]
 dyn_black_fluid_lft_05 [{
   shapeType=Capsule
   length=6.4499
   radius=2.5844
   shapeName=dyn_black_fluid_lft_01_capsule
   center=<4.8304 0 -0.0001>
   rotation=<rot 90 -174.9489 -174.9488>
  }]
 dyn_black_fluid_lft_06 [{
   shapeType=Capsule
   length=3.9499
   radius=2.5844
   shapeName=dyn_black_fluid_lft_01_capsule
   center=<2.6345 0 0>
   rotation=<rot 90 -5.0511 -5.0511>
  }]
 dyn_black_fluid_rgt_01 [{
   shapeType=Capsule
   length=5.2827
   radius=2.6894
   shapeName=dyn_black_fluid_rgt_01_capsule
   center=<4.3499 0.0234 0.0639>
   rotation=<rot 2.7138 92.7916 -94.398>
  }]
 dyn_black_fluid_rgt_03 [{
   shapeType=Capsule
   length=5.2827
   radius=2.6894
   shapeName=dyn_black_fluid_rgt_01_capsule
   center=<4.3499 0.0234 0.0639>
   rotation=<rot 2.7138 92.7916 -94.398>
  }]
 dyn_black_fluid_rgt_05 [{
   shapeType=Capsule
   length=5.2827
   radius=2.6894
   shapeName=dyn_black_fluid_rgt_01_capsule
   center=<4.3499 0.0234 0.0639>
   rotation=<rot 2.7138 92.7916 -94.398>
  }]
 pack [{shapeType=Capsule, length=0.505, radius=0.505, shapeName=pack_capsule, rotation=<rot 0 -90 0>}]
}
constraints [
 {bone1Name=dyn_black_fluid_lft_01, bone2Name=pack, swing1LimitDegrees=30, swing2LimitDegrees=15, twistLimitDegrees=1}
 {bone1Name=dyn_black_fluid_lft_02, bone2Name=dyn_black_fluid_lft_01, swing1LimitDegrees=30, swing2LimitDegrees=15, twistLimitDegrees=1}
 {bone1Name=dyn_black_fluid_lft_05, bone2Name=dyn_black_fluid_lft_02, swing1LimitDegrees=30, swing2LimitDegrees=15, twistLimitDegrees=1}
 {bone1Name=dyn_black_fluid_lft_06, bone2Name=dyn_black_fluid_lft_05, swing1LimitDegrees=30, swing2LimitDegrees=15, twistLimitDegrees=1}
 {bone1Name=dyn_black_fluid_rgt_01, bone2Name=pack, swing1LimitDegrees=30, swing2LimitDegrees=15, twistLimitDegrees=1}
 {bone1Name=dyn_black_fluid_rgt_03, bone2Name=dyn_black_fluid_rgt_01, swing1LimitDegrees=30, swing2LimitDegrees=15, twistLimitDegrees=1}
 {bone1Name=dyn_black_fluid_rgt_05, bone2Name=dyn_black_fluid_rgt_03, swing1LimitDegrees=30, swing2LimitDegrees=15, twistLimitDegrees=1}
]
