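# Animation Blueprint for the Psyche backpack, adding physics sway and liquid deformation to its bones.
# A RigidBody simulation from reference pose feeds four additive Z-scale bone modifiers on the left and right liquid bones.
# Each update maps liquid-bone pitch into two shared alphas, left then right, so right-side values overwrite left.
# A second RigidBody, a Control Rig node and roll-based logic are left unconnected.
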AnimBlueprint /BRCosmetics/Accessories/FORT_Backpacks/Backpack_M_MED_Psyche/Meshes/M_MED_Psyche_Pack_AnimBP
target_skeleton=&/M_MED_Psyche_Pack_Skeleton
graphs [
 AnimGraph {nodes [
   n1 = RigidBody() {
    bClampLinearTranslationLimitToRefPose=true
    bOverrideWorldGravity=true
    overridePhysicsAsset=/Game/Accessories/FORT_Backpacks/Backpack_M_MED_Psyche/Meshes/M_MED_Psyche_Pack_Physics_2
    simulationSpace=BaseBoneSpace
    alphaBoolBlend {blendOption=Linear}
    baseBoneRef {boneName=pack}
    componentAppliedLinearAccClamp=<1500 1500 1500>
    componentLinearAccScale=<0.2 0.2 0.3>
    componentLinearVelScale=<0.1 0.1 0.15>
    overrideWorldGravity {z=200}
   }
   Component To Local(ComponentPose=@n1)
   Control Rig(Source=Component To Local()) {alphaBoolBlend {blendOption=Linear}, controlRigAssetReference=~}
   n2 = RigidBody(ComponentPose=Local To Component(LocalPose=Local Space Ref Pose())) {
    bClampLinearTranslationLimitToRefPose=true
    bOverrideWorldGravity=true
    cachedBoundsScale=1.1
    overridePhysicsAsset=&/M_MED_Psyche_Pack_Physics
    simulationSpace=BaseBoneSpace
    alphaBoolBlend {blendOption=Linear}
    baseBoneRef {boneName=root}
    componentAppliedLinearAccClamp=<3000 3000 3000>
    componentLinearAccScale=<0.5 0.5 0.5>
    componentLinearVelScale=<0.334 0.334 0.334>
    overrideWorldGravity {z=-980}
   }
   n3 = "Transform (Modify) Bone - Bone: dyn_liquid_psyche_lft"(
    ComponentPose=Local To Component(LocalPose=Component To Local(ComponentPose=@n2))
    Alpha=Get Liquid Trans() -> (Liquid Trans:double)
    Translation="0,0,0"
    Rotation="0,0,0"
    Scale="1,1,1.15"
   ) {scaleMode=BMM_Additive, alphaBoolBlend {blendOption=Linear}, boneToModify {boneName=dyn_liquid_psyche_lft}, scale {z=1.15}}
   n4 = "Transform (Modify) Bone - Bone: dyn_liquid_psyche_lft"(ComponentPose=@n3, Alpha=Get Liquid Scale() -> (Liquid Scale:double), Translation="0,0,0", Rotation="0,0,0", Scale="1,1,1.04") {scaleMode=BMM_Additive, alphaBoolBlend {blendOption=Linear}, boneToModify {boneName=dyn_liquid_psyche_lft}, scale {z=1.04}}
   n5 = "Transform (Modify) Bone - Bone: dyn_liquid_psyche_rgt"(ComponentPose=@n4, Alpha=Get Liquid Trans(), Translation="0,0,0", Rotation="0,0,0", Scale="1,1,1.05") {scaleMode=BMM_Additive, alphaBoolBlend {blendOption=Linear}, boneToModify {boneName=dyn_liquid_psyche_rgt}, scale {z=1.05}}
   n6 = "Transform (Modify) Bone - Bone: dyn_liquid_psyche_rgt"(ComponentPose=@n5, Alpha=Get Liquid Scale(), Translation="0,0,0", Rotation="0,0,0", Scale="1,1,1.15") {scaleMode=BMM_Additive, alphaBoolBlend {blendOption=Linear}, boneToModify {boneName=dyn_liquid_psyche_rgt}, scale {z=1.15}}
   Output Pose(Result=Component To Local(ComponentPose=@n6)) {layerGroup=None}
  ]}
 EventGraph {nodes [
   Event BlueprintUpdateAnimation()
   > n1 = Sequence!(execute) -> (
    then_0
    then_1
    then_2:Exec
    then_3:Exec
    then_4:Exec
    then_5:Exec
    then_6:Exec
    then_7:Exec
    then_8:Exec
    then_9:Exec
    then_10:Exec
    then_11:Exec
   )
   n2 = GetDeltaTransformFromRefPose(self=GetOwningComponent(), BoneName=dyn_liquid_psyche_lft, BaseName=dyn_master_psyche_lft)
   > "Set Y(Pitch)"("Y(Pitch)":double=BreakRotator(InRot=BreakTransform(InTransform=@n2).Rotation).Pitch) -> (Output_Get:double)
   Set Liquid Scale(execute=@n1.then_1, Liquid Scale:double=MapRangeClamped(Value="Get Y(Pitch)"() -> ("Y(Pitch)":double), InRangeA=1, InRangeB=20)) -> (Output_Get:double)
   Set Liquid Trans(execute=@n1.then_2, Liquid Trans:double=MapRangeClamped(Value="Get Y(Pitch)"(), InRangeA=-1, InRangeB=-20)) -> (Output_Get:double)
   n3 = GetDeltaTransformFromRefPose(self=GetOwningComponent(), BoneName=dyn_liquid_psyche_rgt, BaseName=dyn_master_psyche_rgt)
   "Set Y(Pitch)"(execute=@n1.then_3, "Y(Pitch)"=BreakRotator(InRot=BreakTransform(InTransform=@n3).Rotation).Pitch)
   Set Liquid Scale(execute=@n1.then_4, Liquid Scale=MapRangeClamped(Value="Get Y(Pitch)"(), InRangeA=1, InRangeB=20))
   Set Liquid Trans(execute=@n1.then_5, Liquid Trans=MapRangeClamped(Value="Get Y(Pitch)"(), InRangeA=-1, InRangeB=-20))
   TryGetPawnOwner()
   "Absolute (Float)"(A=0)
   TryGetPawnOwner()
   TryGetPawnOwner()
   Branch(Condition="float > float"(A=0, B=0) {is=title K2Node_CallFunction})
   n4 = GetDeltaTransformFromRefPose(self=GetOwningComponent(), BoneName=dyn_liquid_psyche_rgt, BaseName=dyn_master_psyche_rgt)
   "Set X(Roll)"("X(Roll)":double=BreakRotator(InRot=BreakTransform(InTransform=@n4).Rotation)) -> (Output_Get:double)
   Set Liquid Scale(Liquid Scale=0)
   n5 = GetDeltaTransformFromRefPose(self=GetOwningComponent(), BoneName=dyn_liquid_psyche_lft, BaseName=dyn_master_psyche_lft)
   "Set X(Roll)"("X(Roll)"=BreakRotator(InRot=BreakTransform(InTransform=@n5).Rotation))
   Set Liquid Scale(Liquid Scale=MapRangeClamped(Value="Get X(Roll)"() -> ("X(Roll)":double), InRangeA=1, InRangeB=20))
   Set Liquid Scale(Liquid Scale=0)
   Set Liquid Trans(Liquid Trans=MapRangeClamped(Value="Get X(Roll)"(), InRangeA=-1, InRangeB=-20))
   Set Liquid Trans(Liquid Trans=MapRangeClamped(Value="Get X(Roll)"(), InRangeA=0, InRangeB=-1))
   Set Liquid Scale(Liquid Scale=MapRangeClamped(Value="Get X(Roll)"(), InRangeA=0, InRangeB=1))
  ]}
]
variables [
 Liquid Scale {type=number}
 Liquid Trans {type=number}
 "X(Roll)" {type=number}
 "Y(Pitch)" {type=number}
]
